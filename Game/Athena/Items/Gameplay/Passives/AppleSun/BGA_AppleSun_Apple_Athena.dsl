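; Crash pad gameplay actor (AppleSun) that bounces pawns, vehicles, projectiles and physics objects, with material-driven wobble, bounce and deflate effects.
; BeginPlay feeds hotfixable values into native crashpad collision handling, keeping deprecated Blueprint launch logic as a hotfix fallback.
; A randomized timer can re-toggle the collider to catch objects inside, and launched actors get temporary tags.
; Lifespan expiry or loss of the linked support triggers a warning, then destruction.
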(event EventBeginPlay
  (bind _pad (Variables|Default|GetPad))
  (|CallParentFunction)
  (Utilities|FlowControl|Sequence
    (:then_0
      (Networking|SwitchHasAuthority
        (:Authority
          ;[1 1 1] Lifespan timer
          (if (Variables|Default|GetHasLifespan)
            (bind _returnvalue (Utilities|Time|SetTimerbyEvent (EventDispatchers|CreateEvent :!event Custom|DestroyPad) (ScalableFloat|GetValueatLevel (Variables|HotfixValues|GetLifespan_))))
            (Variables|Default|SetLifespanTimer _returnvalue)))
        (:Remote
          (Utilities|FlowControl|Sequence
            (:then_0
              ;[1 1 1] Spawn audio and animation
              (if (not (Variables|Default|GetIsWorldPlaced))
                (Audio|PlaySoundatLocation "/Game/Athena/Sounds/Gadgets/AppleSun/Applesun_Deploy_Cue.Applesun_Deploy_Cue" (Transformation|GetActorLocation))
                (Rendering|Material|SetScalarParameterValueonMaterials (Variables|Default|GetPad) "WobbleStartTime" (Utilities|Time|GetGameTimeinSeconds))
                (GameplayCue|ExecuteGameplayCue%28Burst%29Local self (Variables|Default|GetSpawnFeedbackGC) (Ability|GameplayCue|MakeGameplayCueParameters 0.0))
                (Components|Activation|Activate (Variables|Default|GetP_AppleSun_Spawn) false)))
            (:then_1
              ;[1 1 1] Start despawn warning timer
              (if (Variables|Default|GetHasLifespan)
                (bind _returnvalue_2 (Utilities|Time|SetTimerbyEvent (EventDispatchers|CreateEvent :!event Custom|DespawnTell) (- (ScalableFloat|GetValueatLevel (Variables|HotfixValues|GetLifespan_)) (ScalableFloat|GetValueatLevel (Variables|HotfixValues|GetTimeoutTellDuration_)))))
                (Variables|Default|SetTimeoutTellTimer _returnvalue_2)
                (else
                  _))
              ;[1 1 1] Get water info
              (bind (_outhit _returnvalue_3) (Collision|LineTraceForObjects (+ (Transformation|GetActorLocation) "0, 0,100.000000") (- (Transformation|GetActorLocation) "0, 0,100.000000") (Utilities|Array|MakeArray "ObjectTypeQuery13")))
              (if _returnvalue_3
                (bind _asfort_water_body_actor (Utilities|Casting|CastToFortWaterBodyActor (.HitActor (Collision|BreakHitResult _outhit))))
                (Variables|Default|SetWaterLocation (.Location (Collision|BreakHitResult _outhit)))
                (bind _outwatersurfaceinfo (WaterBody|GetWaterSurfaceInfoStruct _asfort_water_body_actor (.Location (Collision|BreakHitResult _outhit)) true true false))
                (Rendering|Material|SetScalarParameterValueonMaterials _pad "WaterDepth" (.WaterDepth (Utilities|Struct|BreakWaterSurfaceInfo _outwatersurfaceinfo)))
                (Rendering|Material|SetScalarParameterValueonMaterials _pad "WaterIndex" (Math|Conversions|ToFloat(Integer) (.WaterBodyIdx (Utilities|Struct|BreakWaterSurfaceInfo _outwatersurfaceinfo))))
                (bind _returnvalue_4 (Effects|Components|ParticleSystem|SpawnFXAtLocation "/Game/Athena/Items/Gameplay/Passives/AppleSun/NS_WaterSplash_AppleSun.NS_WaterSplash_AppleSun" (.WaterSurfaceLocation (Utilities|Struct|BreakWaterSurfaceInfo _outwatersurfaceinfo))))
                (Audio|PlaySoundatLocation "/Game/Athena/Sounds/Gadgets/AppleSun/Applesun_Splash_Cue.Applesun_Splash_Cue" (Transformation|GetActorLocation))
                (Effects|Components|ParticleSystem|SetFloatParameter _returnvalue_4 "User.WaterDepth" (.WaterDepth (Utilities|Struct|BreakWaterSurfaceInfo _outwatersurfaceinfo)))
                (Effects|Components|ParticleSystem|SetVectorParameter _returnvalue_4 "User.WaterPlaneLocation" (Utilities|Struct|BreakWaterSurfaceInfo _outwatersurfaceinfo))
                (Effects|Components|ParticleSystem|SetVectorParameter _returnvalue_4 "User.WaterPlaneNormal" (.WaterPlaneNormal (Utilities|Struct|BreakWaterSurfaceInfo _outwatersurfaceinfo)))))))))
    (:then_1
      (LinkToActorComponent|PerformLinkingCheck (Variables|Default|GetFortLinkToActor) "Down" (Variables|Default|GetLinkedActor)))
    (:then_2
      (Utilities|FlowControl|Delay 0.2)
      (Collision|SetCollisionEnabled (Variables|Default|GetSidesPawnBlocker) "QueryAndPhysics") ; Can't have this on when spawned in case it spawns on top of another player.
    )
    (:then_3
      ;[1 1 1] Start the collision timer if needed and have it tick with a bit of a random variance so that it most likely will never tick on the same frame if there are multiple.
      (if (Variables|Default|GetUsesCollisionTimer)
        (bind _returnvalue_5 (Utilities|Time|SetTimerbyEvent (delegate Custom|CatchObjectsInside) (Math|Random|RandomFloatinRange 2.0 3.0) true))
        (Variables|Default|SetCollisionCheckTimerHandle _returnvalue_5)))
    (:then_4
      (Variables|Default|SetCachedUseNativeCollisionHandling (!= (ScalableFloat|GetValueatLevel (Variables|Crashpad|GetUseNativeCollisionHandling)) 0))
      (Variables|BuildingGameplayActorCrashpad|SetStaticParams (Utilities|Struct|MakeCrashpadStaticParams "(TagName=\"MovementMode.Hover\")" (ScalableFloat|GetValueatLevel (Variables|HotfixValues|GetHorizontalLaunch)) 1.4 1.0 1.05 (ScalableFloat|GetValueatLevel (Variables|HotfixValues|GetMinLaunch)) (ScalableFloat|GetValueatLevel (Variables|HotfixValues|GetMaxLaunch)) (ScalableFloat|GetValueatLevel (Variables|Default|GetMinLaunch_Crashpad+)) (ScalableFloat|GetValueatLevel (Variables|Default|GetMaxLaunch_Crashpad+)) (ScalableFloat|GetValueatLevel (Variables|HotfixValues|GetGrowthStart)) (ScalableFloat|GetValueatLevel (Variables|HotfixValues|GetGrowthEnd)) (Variables|Default|GetLaunchCurve) (Variables|Default|GetLaunchAngleWeight) (Variables|Default|GetPlaced%20Launch%20Mod) (ScalableFloat|GetValueatLevel (Variables|HotfixValues|GetVehicleLaunchMultiplierXY)) (ScalableFloat|GetValueatLevel (Variables|HotfixValues|GetVehicleLaunchMultiplierZ)) 777.0 1.0 1.0 0.22 (Variables|Default|GetCrashpadBounceScales) 0.01 (ScalableFloat|GetValueatLevel (Variables|HotfixValues|GetProjectileLaunchMultiplierXY)) (ScalableFloat|GetValueatLevel (Variables|HotfixValues|GetProjectileLaunchMultiplierZ)) 0.22 (Variables|Default|GetFallImmunityGameplayEffect) (Variables|Default|GetWebLaunchGameplayEffect) "(TagName=\"Pawn.Athena.NPC.Wildlife.SpicySake\")" (Variables|Default|GetDestroyByShark) "(TagName=\"Gameplay.Status.KnockbackImmunity\")" (Variables|Default|GetTag_BuildingPhysics) (Variables|Default|GetPhysicsObjectMassMultiplier) (ScalableFloat|GetValueatLevel (Variables|Default|GetRow_PhysicsObjectImpulseMult)) 0.5 5.0 (ScalableFloat|GetValueatLevel (Variables|HotfixValues|GetMaxLaunch)) -1e+05 0.0 0.22 1.65 1.65 0.01 (Variables|Default|GetMinNonPawnVelocity) (Variables|Default|GetMaxNonPawnVelocity) (Variables|Default|GetTag_EventUsedLaunchPad) 0.22 (Variables|Default|GetPad) (Variables|Default|GetTC_Crashpad+) (Variables|Quest|GetT_Quest_AppleSun_Use)))
      (if (Variables|Default|GetCachedUseNativeCollisionHandling)
        (Components|DestroyComponent (Variables|Default|GetPlayerDetector))))))

(event OnComponentBeginOverlap(Box)
  (params (OverlappedComponent :type Primitive%20Component%20Object%20Reference) (OtherActor :type Actor%20Object%20Reference) (OtherComp :type Primitive%20Component%20Object%20Reference) (OtherBodyIndex :type Integer) (bFromSweep :type Boolean) (SweepResult :type Hit%20Result%20Structure%20%28by%20ref%29))
  (Variables|Default|SetCollidingActor OtherActor)
  (CallFunction|DoCollisionHandling)
  (Variables|Default|SetCollidingActor 0))

(event OnComponentHit(Box)
  (params (HitComponent :type Primitive%20Component%20Object%20Reference) (OtherActor :type Actor%20Object%20Reference) (OtherComp :type Primitive%20Component%20Object%20Reference) (NormalImpulse :type Vector :default "0, 0, 0") (Hit :type Hit%20Result%20Structure%20%28by%20ref%29))
  (Variables|Default|SetCollidingActor OtherActor)
  (CallFunction|DoCollisionHandling)
  (Variables|Default|SetCollidingActor 0))

(event Custom|ServerDestroyPad
  (params (NewParam :type Boolean))
  (CallFunction|DestroyPad))

(event Custom|DoCollisionHandling
  (bind _returnvalue (Actor|GetComponentbyClass (Variables|Default|GetCollidingActor) "/Script/Engine.ProjectileMovementComponent"))
  (bind (_asfort_pickup_athena _bsuccess) (Utilities|Casting|CastToFortPickupAthena (Variables|Default|GetCollidingActor)))
  (bind (_asfort_projectile_petrol _bsuccess_2) (Utilities|Casting|CastToFortProjectilePetrol (Variables|Default|GetCollidingActor)))
  (bind (_returnvalue_x _returnvalue_y _returnvalue_z) (Math|Vector|BreakVector (Utilities|GetNonPawnBounceVel (Variables|Default|GetCollidingActor))))
  (bind (_returnvalue_x_2 _returnvalue_y_2 _returnvalue_z_2) (Math|Vector|BreakVector (Utilities|GetNonPawnBounceVel (Variables|Default|GetCollidingActor))))
  (bind _collidingactor (Variables|Default|GetCollidingActor))
  (bind _collidingactor_2 (Variables|Default|GetCollidingActor))
  (bind _collidingactor_3 (Variables|Default|GetCollidingActor))
  (bind _collidingactor_4 (Variables|Default|GetCollidingActor)
    (:then
      (Ability|GameplayEffect|AddLooseGameplayTags _collidingactor_4 (Variables|Default|GetRecentlyLaunchedGameplayTags))
      (Utilities|Array|Add (Variables|Default|GetRecentlyLaunchedActors) _collidingactor_4)
      (if (Variables|Default|GetCachedUseNativeCollisionHandling)
        (bind _asfort_projectile_base (Utilities|Casting|CastToFortProjectileBase (Variables|Default|GetCollidingActor))
          (:then
            (if (Variables|/Script/FortniteGame.FortProjectileBase|GetbUseClientsidePrediction _asfort_projectile_base)
              (CallFunction|ServerDestroyPad true)
              (else
                _)))
          (:CastFailed
            _))
        (bind _asb_prj_athena_apple_sun (Utilities|Casting|CastToB_Prj_Athena_AppleSun_ _collidingactor_3)
          (:then
            _)
          (:CastFailed
            (CallFunction|CheckforBallercoaster _collidingactor_3) ; In case of collision with object riding Ballercoaster -- Destroy this Actor\nFixes FORT-478190
          ))
        (Utilities|Struct|SetmembersinCrashpadDynamicParams (Variables|Default|GetHandleCollisionDynamicParams) (Variables|Default|GetCollidingActor) (Variables|Default|GetTrackedPlayers) (Utilities|IsValid _asb_prj_athena_apple_sun :!class /Script/Engine.KismetSystemLibrary))
        (Crashpad|HandleCollision (Variables|Default|GetHandleCollisionDynamicParams))
        ;[1 1 1] Quest event to grab the BGA for a quest
        (Utilities|IsValid (Ability|GetAbilitySystemComponent _collidingactor_2)
          (:Is%20Valid
            (Ability|SendGameplayEventtoActor _collidingactor_2 (Variables|Default|GetEvent%20Tag_0) (Utilities|Struct|MakeGameplayEventData (Variables|Default|GetEvent%20Tag_0) self _collidingactor_2)) ; Let AI creatures know they've been lauched
          ))
        (else
          ;[0 0 0] DEPRECATED - soft deprecated (i.e. only used as a hotfix controlled backup) in favor of native collision handling
          (Variables|Default|SetBounceLocation (Transformation|GetActorLocation (Variables|Default|GetCollidingActor)))
          (bind _asfort_pawn (Utilities|Casting|CastToFortPawn (Variables|Default|GetCollidingActor))
            (:then
              ;;[1 1 1] Launching Players
              (Utilities|IsinVehicle _asfort_pawn
                (:false
                  (if (not (GameplayTags|HasMatchingGameplayTag _asfort_pawn "(TagName=\"MovementMode.Hover\")"))
                    (bind _returnvalue_2 (CallFunction|GetLaunchReferenceVector _asfort_pawn))
                    (bind _launchedpawn (Utilities|LaunchPawns (Variables|Actor|GetInstigator) _asfort_pawn (Utilities|CalculateVelocity _returnvalue_2 (* (Utilities|Map|Find (Variables|Default|GetTrackedPlayers) _asfort_pawn) 1.05)) true true false true true false false true))
                    (Utilities|Untether _launchedpawn)
                    ;;[1 0 0.1225] Is there a particular reason we don't want to send this event to non-AI pawns?
                    (Ability|SendGameplayEventtoActor _launchedpawn (Variables|Default|GetTag_EventUsedLaunchPad)) ; Let AI creatures know they've been launched
                  ))))
            (:CastFailed
              ;;[1 1 1] Launching Vehicles
              (bind _asfort_vehicle_interface (Utilities|Casting|CastToFortVehicleInterface _collidingactor)
                (:then
                  (Audio|BroadcastAbilityImpactatLocation 0 (Utilities|ToObject(Interface) _asfort_vehicle_interface) (Math|Vector|MakeVector (* _returnvalue_x_2 (ScalableFloat|GetValueatLevel (Variables|HotfixValues|GetVehicleLaunchMultiplierXY))) (* _returnvalue_y_2 (ScalableFloat|GetValueatLevel (Variables|HotfixValues|GetVehicleLaunchMultiplierXY))) (* _returnvalue_z_2 (ScalableFloat|GetValueatLevel (Variables|HotfixValues|GetVehicleLaunchMultiplierZ)))) 777.0))
                (:CastFailed
                  ;;[1 1 1] If hit a physics object, destroy
                  (if (Utilities|IsValid (Actor|GetComponentbyClass _collidingactor "/Script/FortniteGame.FortPhysicsObjectComponent") :!class /Script/Engine.KismetSystemLibrary)
                    (CallFunction|DestroyPad)
                    (else
                      ;;[1 1 1] Launching Other Crash Pads
                      (bind _asb_prj_athena_apple_sun_2 (Utilities|Casting|CastToB_Prj_Athena_AppleSun_ _collidingactor)
                        (:then
                          (Variables|/Script/Engine.MovementComponent|SetVelocity (* (+ (Math|Vector|Normalize (Math|Vector|ProjectVectorontoPlane (Transformation|GetVelocity _asb_prj_athena_apple_sun_2) "0, 0,1.000000")) "0, 0,1.000000") (Math|Float|SelectFloat (Utilities|Select :Index (Variables|/Game/Athena/Items/Gameplay/Passives/AppleSun/B_Prj_Athena_AppleSun_.B_Prj_Athena_AppleSun__C|GetCrashpadsBouncedOffOf _asb_prj_athena_apple_sun_2) :Option%200 1.7e+03 :Option%201 1.95e+03 :Option%202 2125.0 :Option%203 2.2e+03 :Option%204 2.26e+03 :Option%205 2.29e+03) 2.29e+03 (<= (Variables|/Game/Athena/Items/Gameplay/Passives/AppleSun/B_Prj_Athena_AppleSun_.B_Prj_Athena_AppleSun__C|GetCrashpadsBouncedOffOf _asb_prj_athena_apple_sun_2) 5))) (Variables|/Script/FortniteGame.FortProjectileBase|GetProjectileMovementComponent _asb_prj_athena_apple_sun_2))
                          (Math|Integer|IncrementInt (Variables|/Game/Athena/Items/Gameplay/Passives/AppleSun/B_Prj_Athena_AppleSun_.B_Prj_Athena_AppleSun__C|GetCrashpadsBouncedOffOf _asb_prj_athena_apple_sun_2)
                            (:%20%20
                              (CallFunction|BounceMulticast (Variables|Default|GetBounceLocation) 0.22 (Variables|Default|GetCollidingActor)))))
                        (:CastFailed
                          ;;[1 1 1] Launching Projectiles
                          (if (Math|Boolean|ANDBoolean (Utilities|IsValid _returnvalue :!class /Script/Engine.KismetSystemLibrary) (not _bsuccess_2) (not _bsuccess))
                            ;;[1 1 1] If the actor also has a Held Object component, only bounce if it is moving and not being held FORT-274357
                            (Utilities|IsValid (Actor|GetComponentbyClass _collidingactor "/Script/FortniteGame.FortHeldObjectComponent")
                              (:Is%20Valid
                                (bind _returnvalue_3 (Fortnite|HeldObjects|IsHeldObjectBeingHeld _collidingactor))
                                (if (and (not _returnvalue_3) (Math|Vector|NotEqual(Vector) (Transformation|GetVelocity _collidingactor) "0, 0, 0" 0.01))
                                  _))
                              (:Is%20Not%20Valid
                                _))
                            (Variables|/Script/Engine.MovementComponent|SetVelocity (Math|Vector|MakeVector (* _returnvalue_x (ScalableFloat|GetValueatLevel (Variables|HotfixValues|GetProjectileLaunchMultiplierXY))) (* _returnvalue_y (ScalableFloat|GetValueatLevel (Variables|HotfixValues|GetProjectileLaunchMultiplierXY))) (* _returnvalue_z (ScalableFloat|GetValueatLevel (Variables|HotfixValues|GetProjectileLaunchMultiplierZ)))) _returnvalue)
                            (CallFunction|BounceMulticast (Variables|Default|GetBounceLocation) 0.22 (Variables|Default|GetCollidingActor))
                            (else
                              (CallFunction|LaunchBuldingActors _collidingactor)))))))))))
          (Utilities|FlowControl|Sequence
            (:then_0
              (Networking|SwitchHasAuthority
                (:Authority
                  (Networking|SetNetDormancy "DORM_Awake") ; Added to fix [FORT-404291]
                  (CallFunction|BounceMulticast (Variables|Default|GetBounceLocation) 1.0 (Variables|Default|GetCollidingActor)))))
            (:then_1
              (GameplayEffects|ApplyGameplayEffectToTarget :self (Ability|GetAbilitySystemComponent _launchedpawn) :GameplayEffectClass (Variables|Default|GetFallImmunityGameplayEffect) :Target (Ability|GetAbilitySystemComponent _launchedpawn) :Level 0.0)
              (GameplayEffects|ApplyGameplayEffectToTarget :self (Ability|GetAbilitySystemComponent _launchedpawn) :GameplayEffectClass (Variables|Default|GetWebLaunchGameplayEffect) :Target (Ability|GetAbilitySystemComponent _launchedpawn) :Level 0.0)
              (Utilities|IsValid (Fortnite|GetFortPlayerControllerfromActor _launchedpawn)
                (:Is%20Valid
                  (Networking|SwitchHasAuthority
                    (:Authority
                      (Quest|SendComplexCustomStatEvent (Quests|GetQuestManager (Fortnite|GetFortPlayerControllerfromActor _launchedpawn) "Invalid") 0 (Variables|Quest|GetT_Quest_AppleSun_Use) "(GameplayTags=)" 1))))))
            (:then_2
              (if (and (GameplayTags|HasMatchingGameplayTag _launchedpawn "(TagName=\"Pawn.Athena.NPC.Wildlife.SpicySake\")") (Variables|Default|GetDestroyByShark))
                (CallFunction|DestroyPad)))))))
    (:else)))

(event Custom|CatchObjectsInside
  ;[1 1 1] Deactivate and reactivate collider so that On Begin Overlap event fires again for anything that got caught inside.
  (Collision|SetCollisionEnabled (Variables|Default|GetBox) "NoCollision")
  (Utilities|FlowControl|Delay 0.0)
  (Collision|SetCollisionEnabled (Variables|Default|GetBox) "QueryAndPhysics"))

(event Custom|DespawnTell
  ;[1 1 1] Leaking air FX warning when about to despawn
  (Components|Activation|SetActive (Variables|Default|GetP_AppleSun_DespawnTell) true false)
  (Audio|Components|Audio|FadeIn (Variables|Default|GetAppleSun_Warning_Loop_01) 0.05 1.0 0.0 "Linear" :!class /Script/Engine.AudioComponent)
  (bind (_direction _newtrack_0) (Timeline|BeginDeflate PlayFromStart)
    (:Update
      (Rendering|Material|SetScalarParameterValueonMaterials (Variables|Default|GetPad) "DeflateStart" _newtrack_0))))

(event Custom|BounceMulticast
  (params (Position :type Vector) (BounceIntensity :type Float%20%28double-precision%29) (Actor :type Actor%20Object%20Reference))
  (bind _pad (Variables|Default|GetPad))
  ;[1 1 1] Bounce animation and audio
  (Development|PrintString (Utilities|String|Append "BounceMultiCast Position: " (Utilities|String|ToString(Vector) Position) "  Intensity:  " (Utilities|String|ToString(Float) BounceIntensity)) false)
  (Networking|CanExecuteCosmeticEvents
    (:True
      (Rendering|Material|SetScalarParameterValueonMaterials _pad "BounceStartTime" (Utilities|Time|GetGameTimeinSeconds))
      (Rendering|Material|SetVectorParameterValueonMaterials _pad "BouncePosition" Position)
      (Rendering|Material|SetScalarParameterValueonMaterials _pad "BounceIntensity" BounceIntensity)
      (CallFunction|PlayBounceAudio Actor)
      (Development|PrintString (Utilities|String|BuildString(Vector) "" "BounceMultiClass Position" Position) false)
      (Development|PrintString (Utilities|String|BuildString(Float) "" "BounceMuliCast Intensity: " BounceIntensity) false))))

(event Crashpad|EventDoBounceMulticast
  (params (BounceLocation :type Vector :default "0, 0, 0") (BounceIntensity :type Float%20%28single-precision%29) (BouncingActor :type Actor%20Object%20Reference))
  (CallFunction|BounceMulticast BounceLocation BounceIntensity BouncingActor))

(event EventOnDeathPlayEffects
  (params (Damage :type Float%20%28single-precision%29) (DamageTags :type Gameplay%20Tag%20Container%20Structure%20%28by%20ref%29) (Momentum :type Vector :default "0, 0, 0") (HitInfo :type Hit%20Result%20Structure%20%28by%20ref%29) (InstigatedBy :type Fort%20Pawn%20Object%20Reference) (DamageCauser :type Actor%20Object%20Reference) (EffectContext :type Gameplay%20Effect%20Context%20Handle%20Structure))
  (bind (_returnvalue_x _returnvalue_y _returnvalue_z) (Math|Vector|BreakVector (Math|Transform|InverseTransformLocation (Transformation|GetActorTransform) (Variables|Default|GetWaterLocation))))
  (bind _pad (Variables|Default|GetPad))
  (bind _death_effect_template (Variables|FX|GetDeath%20Effect%20Template))
  ;[1 1 1] Destruction FX and audio
  (Components|Activation|SetActive (Variables|Default|GetP_AppleSun_DespawnTell) false false)
  (Utilities|Time|ClearandInvalidateTimerbyHandle (Variables|Default|GetTimeoutTellTimer))
  (Utilities|FlowControl|DoOnce false
    (:Completed
      (if (Variables|FX|GetDeath%20Effect%20At%20Loction)
        (bind _returnvalue (Effects|Components|ParticleSystem|SpawnFXAtLocation _death_effect_template (Transformation|GetActorLocation) (Transformation|GetActorRotation)))
        (Variables|FX|SetDeath%20Effect _returnvalue)
        (else
          (bind _returnvalue_2 (Effects|SpawnFXSystemAttached :FXSystemAsset _death_effect_template :AttachToComponent (Variables|Default|GetPad) :AttachPointName "None" :Scale "1.000000,1.000000,1.000000" :LocationType "SnapToTarget" :bAutoDestroy true :PoolingMethod "None" :bAutoActivate true))
          (Variables|FX|SetDeath%20Effect _returnvalue_2)))
      (Audio|PlaySoundatLocation (Variables|Default|GetDestroySound) (Transformation|GetActorLocation) "0, 0, 0" 1.6)
      (Audio|Components|Audio|FadeOut (Variables|Default|GetAppleSun_Warning_Loop_01) 0.25 0.0 "Linear" :!class /Script/Engine.AudioComponent)
      (GameplayCue|ExecuteGameplayCue%28Burst%29Local self (Variables|Default|GetDeathFeedbackGC) (Ability|GameplayCue|MakeGameplayCueParameters 0.0))
      (bind (_direction _deflate _scale) (Timeline|Destroy PlayFromStart)
        (:Update
          (Rendering|Material|SetScalarParameterValueonMaterials _pad "Deflate" _deflate)
          (Transformation|SetRelativeScale3D _pad (* _scale (Variables|Default|GetStartScale)))
          (Transformation|SetRelativeLocation _pad (Math|Vector|MakeVector 0.0 0.0 (Math|Float|Lerp -9e+01 (Math|Float|SelectFloat _returnvalue_z -7e+01 (Math|Vector|NotEqual(Vector) (Variables|Default|GetWaterLocation) "0, 0, 0" 0.0001)) _deflate)) false false))))))

(event Custom|DestroyPad
  (bind (_returnvalue_x _returnvalue_y _returnvalue_z) (Math|Vector|BreakVector (Math|Transform|InverseTransformLocation (Transformation|GetActorTransform) (Variables|Default|GetWaterLocation))))
  (bind _pad (Variables|Default|GetPad))
  (bind _death_effect_template (Variables|FX|GetDeath%20Effect%20Template))
  ;[1 1 1] Destroy at end of lifespan
  (Collision|SetActorEnableCollision false)
  (Utilities|Time|ClearandInvalidateTimerbyHandle (Variables|Default|GetLifespanTimer))
  (Utilities|FlowControl|Sequence
    (:then_0
      ;[1 1 1] Destruction FX and audio
      (Networking|SwitchHasAuthority
        (:Remote
          (Components|Activation|SetActive (Variables|Default|GetP_AppleSun_DespawnTell) false false)
          (Utilities|Time|ClearandInvalidateTimerbyHandle (Variables|Default|GetTimeoutTellTimer))
          (Utilities|FlowControl|DoOnce false
            (:Completed
              (if (Variables|FX|GetDeath%20Effect%20At%20Loction)
                (bind _returnvalue (Effects|Components|ParticleSystem|SpawnFXAtLocation _death_effect_template (Transformation|GetActorLocation) (Transformation|GetActorRotation)))
                (Variables|FX|SetDeath%20Effect _returnvalue)
                (else
                  (bind _returnvalue_2 (Effects|SpawnFXSystemAttached :FXSystemAsset _death_effect_template :AttachToComponent (Variables|Default|GetPad) :AttachPointName "None" :Scale "1.000000,1.000000,1.000000" :LocationType "SnapToTarget" :bAutoDestroy true :PoolingMethod "None" :bAutoActivate true))
                  (Variables|FX|SetDeath%20Effect _returnvalue_2)))
              (Audio|PlaySoundatLocation (Variables|Default|GetDestroySound) (Transformation|GetActorLocation) "0, 0, 0" 1.6)
              (Audio|Components|Audio|FadeOut (Variables|Default|GetAppleSun_Warning_Loop_01) 0.25 0.0 "Linear" :!class /Script/Engine.AudioComponent)
              (GameplayCue|ExecuteGameplayCue%28Burst%29Local self (Variables|Default|GetDeathFeedbackGC) (Ability|GameplayCue|MakeGameplayCueParameters 0.0))
              (Timeline|Destroy PlayFromStart))))))
    (:then_1
      (Actor|SetLifeSpan 2.0))
    (:then_2
      (Networking|SwitchHasAuthority
        (:Authority
          (Fortnite|FortShippingLog self (Utilities|String|ToString(Text) (Utilities|Text|FormatText "NSLOCTEXT(\"[B046D156433C507C328676A482FCDC90]\", \"CAB97A1D4FFECE5129124F8117B1E98B\", \"BGA_AppleSun_Apple_Athena: Crashpad actor [{actor}] destruction started in DestroyPad.\")" self))))))))

(event EventDestroyed
  ;[1 1 1] Destroy at end of lifespan
  (CallFunction|DestroyPad))

(event Crashpad|EventDestroyCrashpad
  ;[0 0 0] DEPRECATED - soft deprecated (i.e. only used as a hotfix controlled backup) in favor of native collision handling
  ;;[1 1 1] Launching Players
  (CallFunction|DestroyPad))

(event Custom|LaunchBuldingActors
  (params (Actor :type Actor%20Object%20Reference))
  (bind _returnvalue (Transformation|GetActorLocation Actor))
  (bind (_returnvalue_x _returnvalue_y _returnvalue_z) (Math|Vector|BreakVector (Transformation|GetVelocity Actor)))
  ;[0 0 0] DEPRECATED - soft deprecated (i.e. only used as a hotfix controlled backup) in favor of native collision handling
  ;;[1 1 1] If the actor has the building actor tag and also a physics component launch it.
  (bind _returnvalue_2 (GameplayTags|HasMatchingGameplayTag(Message) Actor (Variables|Default|GetTag_BuildingPhysics)))
  (if _returnvalue_2
    (Utilities|IsValid (Actor|GetComponentbyClass Actor "/Script/FortniteGame.FortPhysicsObjectComponent")
      (:Is%20Valid
        (bind _asprimitive_component (Utilities|Casting|CastToPrimitiveComponent (Variables|/Script/Engine.Actor|GetRootComponent Actor)))
        (bind (_returnvalue_x_2 _returnvalue_y_2 _returnvalue_z_2) (Math|Vector|BreakVector (* (Math|Vector|Normalize (Transformation|GetActorUpVector)) (* (* (ScalableFloat|GetValueatLevel (Variables|Default|GetRow_PhysicsObjectImpulseMult)) (Math|Float|Clamp(Float) (* (Physics|GetMass _asprimitive_component) (Variables|Default|GetPhysicsObjectMassMultiplier)) 0.5 5.0)) (ScalableFloat|GetValueatLevel (Variables|HotfixValues|GetMaxLaunch))))))
        (Physics|AddImpulseatLocation _asprimitive_component (Math|Vector|MakeVector _returnvalue_x_2 _returnvalue_y_2 (+ _returnvalue_z_2 (Math|Float|Absolute(Float) (Math|Float|Clamp(Float) _returnvalue_z -1e+05 0.0)))) _returnvalue "None")
        (if (Networking|IsServer)
          (CallFunction|BounceMulticast _returnvalue 0.22 (Variables|Default|GetCollidingActor)))))))

(event OnLinkedActorDestroyed(FortLinkToActor)
  (params (DamageCauser :type Actor%20Object%20Reference))
  ;[0.2144 0.5421 0.7917] OnLinkedActorDestroyed
  ; Check if the actor is still colliding with anything and destroy it for "floating" or try to reattach it for future destroyed support objects
  (bind (_outhit _returnvalue) (Collision|BoxTraceByChannel (+ (Transformation|GetActorLocation self) (* (Transformation|GetActorUpVector self) -15.0)) (+ (Transformation|GetActorLocation self) (* (Transformation|GetActorUpVector self) -15.0)) "300.000000,300.000000,70.000000" (Transformation|GetActorRotation self) "TraceTypeQuery10"))
  (if _returnvalue
    ;;[0.2144 0.5421 0.7917] Attempt to link to new actor (won't link to terrain)
    (bind _returnvalue_2 (LinkToActorComponent|PerformLinkingCheck (Variables|Default|GetFortLinkToActor) "Down" (Variables|Default|GetLinkedActor)))
    (if _returnvalue_2
      _
      (else
        _))
    (else
      _))
  (CallFunction|StartSlowTellDestroy))

(event Custom|StartSlowTellDestroy
  (bind _lifespantimer (Variables|Default|GetLifespanTimer))
  ;[0.3177 0.1299 0] StartSlowTellDestroy
  ; Activate the tell that the crashpad is being destroyed and set the timer to destroy the object
  (Utilities|FlowControl|DoOnce false
    (:Completed
      (Networking|SwitchHasAuthority
        (:Authority
          (bind _returnvalue (Utilities|Time|SetTimerbyEvent (EventDispatchers|CreateEvent :!event Custom|DestroyPad) (Math|Float|Min(Float) (ScalableFloat|GetValueatLevel (Variables|HotfixValues|GetTimeoutTellDuration_)) (Utilities|Time|GetTimerRemainingTimebyHandle _lifespantimer))))
          (Utilities|Time|ClearandInvalidateTimerbyHandle _lifespantimer) ; clear previous timer
          (Variables|Default|SetLifespanTimer _returnvalue))
        (:Remote
          (CallFunction|DespawnTell)
          (Utilities|Time|ClearandInvalidateTimerbyHandle (Variables|Default|GetTimeoutTellTimer)))))))

(event OnComponentBeginOverlap(PlayerDetector)
  (params (OverlappedComponent :type Primitive%20Component%20Object%20Reference) (OtherActor :type Actor%20Object%20Reference) (OtherComp :type Primitive%20Component%20Object%20Reference) (OtherBodyIndex :type Integer) (bFromSweep :type Boolean) (SweepResult :type Hit%20Result%20Structure%20%28by%20ref%29))
  ;[0 0 0] DEPRECATED - soft deprecated (i.e. only used as a hotfix controlled backup) in favor of native collision handling
  ;;[1 1 1] Detect Player velocity early for better accuracy when there's lag
  (if (not (Variables|Default|GetCachedUseNativeCollisionHandling))
    (bind _asfort_pawn (Utilities|Casting|CastToFortPawn OtherActor))
    (Variables|/Script/FortniteGame.FortPawn|SetbIgnoreNextFallingDamage true _asfort_pawn)
    (Utilities|Map|Add (Variables|Default|GetTrackedPlayers) _asfort_pawn (Math|Vector|VectorLength (Variables|/Script/Engine.MovementComponent|GetVelocity (Variables|/Script/Engine.Character|GetCharacterMovement _asfort_pawn))))))

(event OnComponentEndOverlap(PlayerDetector)
  (params (OverlappedComponent :type Primitive%20Component%20Object%20Reference) (OtherActor :type Actor%20Object%20Reference) (OtherComp :type Primitive%20Component%20Object%20Reference) (OtherBodyIndex :type Integer))
  ;[0 0 0] DEPRECATED - soft deprecated (i.e. only used as a hotfix controlled backup) in favor of native collision handling
  ;;[1 1 1] Detect Player velocity early for better accuracy when there's lag
  (if (not (Variables|Default|GetCachedUseNativeCollisionHandling))
    (bind _asfort_pawn (Utilities|Casting|CastToFortPawn OtherActor))
    (Variables|/Script/FortniteGame.FortPawn|SetbIgnoreNextFallingDamage false _asfort_pawn)
    (Utilities|Map|Remove (Variables|Default|GetTrackedPlayers) _asfort_pawn)))

(event EventEndPlay
  (params (EndPlayReason :type EEndPlayReason%20Enum :default "Destroyed"))
  (bind _recentlylaunchedactors (Variables|Default|GetRecentlyLaunchedActors))
  ;[0.7865 0.4084 0.08288] Clean up timer handles and clear references when this actor is destroyed.
  (|CallParentFunction EndPlayReason)
  (Utilities|Time|ClearandInvalidateTimerbyHandle (Variables|Default|GetCollisionCheckTimerHandle))
  (Utilities|Time|ClearandInvalidateTimerbyHandle (Variables|Default|GetLifespanTimer))
  (Utilities|Time|ClearandInvalidateTimerbyHandle (Variables|Default|GetTimeoutTellTimer))
  (Components|Timeline|Stop (Variables|BGA_AppleSun_Apple_Athena|GetBeginDeflate))
  (Variables|Default|SetLinkedActor 0)
  (Utilities|Map|Clear (Variables|Default|GetTrackedPlayers))
  (Variables|Default|SetCollidingActor 0)
  (for _array_element _recentlylaunchedactors
    (Utilities|IsValid _array_element
      (:Is%20Valid
        (Ability|GameplayEffect|RemoveLooseGameplayTags _array_element (Variables|Default|GetRecentlyLaunchedGameplayTags)))))
  (Utilities|Array|Clear _recentlylaunchedactors)
  (Networking|SwitchHasAuthority
    (:Authority
      (Fortnite|FortShippingLog self (Utilities|String|ToString(Text) (Utilities|Text|FormatText "NSLOCTEXT(\"[B046D156433C507C328676A482FCDC90]\", \"217D40A14165DDA2D5A6D4B2C95CFA55\", \"BGA_AppleSun_Apple_Athena: EndPlay called for Crashpad actor [{actor}].\")" self))))))

(event OnComponentEndOverlap(Box)
  (params (OverlappedComponent :type Primitive%20Component%20Object%20Reference) (OtherActor :type Actor%20Object%20Reference) (OtherComp :type Primitive%20Component%20Object%20Reference) (OtherBodyIndex :type Integer))
  (Utilities|IsValid OtherActor
    (:Is%20Valid
      (if (Utilities|Array|ContainsItem (Variables|Default|GetRecentlyLaunchedActors) OtherActor)
        (Utilities|Array|RemoveItem (Variables|Default|GetRecentlyLaunchedActors) OtherActor)
        (Ability|GameplayEffect|RemoveLooseGameplayTags OtherActor (Variables|Default|GetRecentlyLaunchedGameplayTags))))))

(fn CheckForBallercoaster
  (params (Colliding%20Actor :type Object%20Reference))
  (bind _returnvalue (GameplayTags|GetOwnedGameplayTags Colliding%20Actor :!class /Script/GameplayTags.GameplayTagAssetInterface))
  (if (GameplayTags|HasTag _returnvalue "(TagName=\"Gameplay.Ballercoaster.Riding\")")
    (Actor|DestroyActor)))

(fn CheckForSpyPerk
  (bind _instigator (Variables|Actor|GetInstigator))
  (Utilities|IsValid _instigator
    (:Is%20Valid
      (return (GameplayTags|TagContainer|CompareTagAssetInterfacetoOtherTagContainers (Variables|Default|GetTC_Crashpad+) (Ability|GetAbilitySystemComponent _instigator))))
    (:Is%20Not%20Valid
      (return false))))

(fn CheckPredictedProjectile)

(fn GetLaunchReferenceVector
  (params (Fort%20Pawn :type Character%20Object%20Reference))
  (bind (_tethered _tetherpawn) (Utilities|CheckforTether Fort%20Pawn)
    (:out
      (if _tethered
        (return (Transformation|GetVelocity _tetherpawn))
        (else
          (return (Pawn|Components|CharacterMovement|GetCurrentAcceleration (Variables|/Script/Engine.Character|GetCharacterMovement Fort%20Pawn))))))))

(fn OnHitPredicted)

(fn PlayBounceAudio
  (params (Actor :type Actor%20Object%20Reference))
  (bind _returnvalue (Transformation|GetActorLocation))
  (bind _3p_bouncesound (Variables|Default|Get3P%20BounceSound))
  (bind _1p_bounce_sound (Variables|Default|Get1P%20Bounce%20Sound))
  (bind _asfort_pawn (Utilities|Casting|CastToFortPawn Actor)
    (:then
      (Audio|PlaySoundatLocation (select (Pawn|IsLocallyViewed _asfort_pawn) _1p_bounce_sound _3p_bouncesound) _returnvalue "0, 0, 0" 1.0 1.0 0.0 0 0 self)
      (Fortnite|ShowSoundIndicatorLocalforTeam :Location _returnvalue :MaxAudibleDistance (Variables|/Script/Engine.SoundBase|GetMaxDistance (select (Pawn|IsLocallyViewed _asfort_pawn) _1p_bounce_sound _3p_bouncesound)) :IndicatorType "Glider" :Team (Class|FortPawn|GetTeam _asfort_pawn) :Affiliations (Utilities|Array|MakeArray "Hostile") :Tint "(R=0.000000,G=0.000000,B=0.000000,A=0.000000)" :OverrideIcon "/Game/UI/Foundation/Textures/Icons/Navigation/SV_Launcher.SV_Launcher" :AttenuationMaxRangeMultiplier 1.0 :AttenuationIntensityMultiplier 1.0)
      (Development|PrintString (Utilities|String|ToString(Boolean) (Pawn|IsLocallyViewed _asfort_pawn)) true true "(R=1.000000,G=0.105670,B=0.000000,A=1.000000)" 5.0))
    (:CastFailed
      (Audio|PlaySoundatLocation (Variables|Default|Get3P%20BounceSound) (Transformation|GetActorLocation) "0, 0, 0" 1.0 1.0 0.0 0 0 self))))

(fn ConstructionScript
  (Variables|Default|SetStartScale (Variables|/Script/Engine.SceneComponent|GetRelativeScale3D (Variables|Default|GetPad))))
