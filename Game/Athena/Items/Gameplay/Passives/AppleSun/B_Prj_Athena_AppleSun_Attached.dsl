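; Attached AppleSun projectile that rides below a falling player and spawns an airbag on the surface ahead.
; A timeline pushes it downward, leading by player velocity, while line traces from the player past the projectile trigger the spawn.
; If the player ziplines or stops falling, a failsafe traces far downward instead.
; Water entry spawns the BGA on authority; a separate surface-validity check filters targets but is only called here for debug drawing.
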(event EventBeginPlay
  (|CallParentFunction)
  (bind (_direction _relativezpos) (Timeline|DistanceFromPlayer Play)
    (:Update
      ; Check Failsafe Case
      (bind _result (CallFunction|IsFailsafeTriggered))
      (Variables|Default|SetbIsFailsafeTriggered _result)
      (CallFunction|UpdateRelativeLocationtoPlayer _relativezpos)
      (Default|TrySpawnAirbag false))))

(event OnEnterWaterEvent(WaterInteractionComponent)
  (params (WaterBody :type Fort%20Water%20Body%20Actor%20Object%20Reference) (WaterInteractionComponent :type Fort%20Water%20Interaction%20Component%20Object%20Reference) (bIsFirstBody :type Boolean))
  (Networking|SwitchHasAuthority
    (:Authority
      (CallFunction|SpawnBGA "0, 0,1.000000" WaterBody 0 (Transformation|GetActorLocation) true))))

(fn IsFailsafeTriggered
  ; Check if the player somehow stops falling without the airbag spawning (e.g. caught the zipline).
  (bind _asfort_player_pawn (Utilities|Casting|CastToFortPlayerPawn (Actor|GetAttachParentActor))
    (:then
      (bind _returnvalue (Zipline|IsZiplining _asfort_player_pawn))
      (return (or _returnvalue (not (GameplayTags|HasMatchingGameplayTag _asfort_player_pawn "(TagName=\"MovementMode.Falling\")")))))
    (:CastFailed
      (return false))))

(fn IsTraceValidForSpawn
  (params (HitOtherComp :type Primitive%20Component%20Object%20Reference) (HitOther :type Actor%20Object%20Reference) (HitResult :type Hit%20Result%20Structure))
  (bind (_bblockinghit _binitialoverlap _time _distance _location _impactpoint _normal _impactnormal _physmat _hitactor) (Collision|BreakHitResult HitResult))
  (bind (_asfort_athena_vehicle _bsuccess) (Utilities|Casting|CastToFortAthenaVehicle HitOther))
  ;[0.474 0.6502 1] IsHitValidForSpawn (Override)
  ; Run through conditions for a valid surface hit and then output the type of action to be taken by the projectile: Invalid (default behavior), Valid (spawns BGA), or Invalid Revert to Pickup (reverts thrown projectile to the pickup version)
  (Utilities|IsValid HitOtherComp
    (:Is%20Valid
      (Utilities|IsValid HitOther
        (:Is%20Valid
          ;;[0.474 0.6502 1] Can stick to the hit target
          (if (not (Components|ComponentHasTag HitOtherComp "NoStick"))
            ;;[0.474 0.6502 1] Won't spawn on objects that can fall
            (Utilities|Casting|CastToBGA_Athena_WithGravity_Parent HitOther
              (:then
                _)
              (:CastFailed
                ;;[0.474 0.6502 1] Won't Spawn on players or held objects
                (bind _returnvalue (GameplayTags|HasMatchingGameplayTag(Message) HitOther "(TagName=\"HeldObject\")"))
                (if (not _returnvalue)
                  (Utilities|Casting|CastToFortPlayerPawn HitOther
                    (:then
                      _)
                    (:CastFailed
                      ;;[0.474 0.6502 1] If spawning on vehicles is allowed, check if vehicle
                      (if (and (Variables|Spawning|ValidityChecks|GetbDontSpawnOnVehicles) _bsuccess)
                        _
                        (elif (> (Math|Vector|DotProduct _normal "0, 0,1.000000") (Variables|Spawning|ValidityChecks|GetMinDotWithUp))
                          ;;[0.474 0.6502 1] Don't land on physics based objects, since those can move via the impact
                          (Utilities|IsValid _hitactor
                            (:Is%20Valid
                              (bind _returnvalue_2 (FortPhysicsSimulationLibrary|CanActorSimulatePhysics _hitactor))
                              (if _returnvalue_2
                                _
                                (elif (CallFunction|CheckIsBlockedbyVolume)
                                  (return false true)
                                  (else
                                    (return true false)))))
                            (:Is%20Not%20Valid
                              _))
                          (return false false))))))))
            (else
              _)))
        (:Is%20Not%20Valid
          _)))
    (:Is%20Not%20Valid
      _)))

(fn TrySpawnAirbag
  (params (bDebug :type Boolean))
  (bind _bcanspawnairbag (LocalVariables|GetbCanSpawnAirbag))
  (bind _returnvalue (Math|Color|SelectColor "(R=1.000000,G=1.000000,B=1.000000,A=1.000000)" "(R=0.270833,G=0.270833,B=0.270833,A=1.000000)" _bcanspawnairbag))
  (bind _hit (LocalVariables|GetHit))
  (bind (_bblockinghit _binitialoverlap _time _distance _location _impactpoint _normal _impactnormal _physmat _hitactor _hitcomponent) (Collision|BreakHitResult _hit))
  ; If the airbag is already spawned, return.
  (if (Variables|Default|GetBGA_Spawned)
    (return)
    (else
      (Utilities|FlowControl|Sequence
        (:then_0
          ; Init
          (LocalVariables|SetAttachParentLocation (+ (Transformation|GetActorLocation (Actor|GetAttachParentActor)) "0.000000,0.000000,10.000000"))
          (LocalVariables|SetProjectileLocation (Transformation|GetActorLocation)))
        (:then_1
          ; Determine trace start & end locations
          (if (Variables|Default|GetbIsFailsafeTriggered)
            ;; Failsafe Trace - Performs longer trace if the player somehow stops falling without the airbag spawning (e.g. caught the zipline).
            (LocalVariables|SetTraceStart (LocalVariables|GetAttachParentLocation))
            (LocalVariables|SetTraceEnd (+ (LocalVariables|GetProjectileLocation) "0, 0,-100000.000000"))
            (else
              ;; Regular Trace - Raycast from player to 200 units past projectile for a spawn location (starts at player just in case a surface comes between them somehow)
              (LocalVariables|SetTraceStart (LocalVariables|GetAttachParentLocation))
              (LocalVariables|SetTraceEnd (+ (LocalVariables|GetProjectileLocation) (* (Utilities|GetDirectionTo (LocalVariables|GetAttachParentLocation) (LocalVariables|GetProjectileLocation)) 2e+02))))))
        (:then_2
          ; If a valid spawn location is found, spawn the airbag.
          (bind (_outhit _returnvalue_2) (Collision|LineTraceByChannel :Start (LocalVariables|GetTraceStart) :End (LocalVariables|GetTraceEnd) :TraceChannel (select (ScalableFloat|GetValueasBool (Variables|Default|GetHF_UseWeaponTrace)) "TraceTypeQuery13" "TraceTypeQuery10") :bTraceComplex false :DrawDebugType "None" :bIgnoreSelf true :TraceColor "(R=1.000000,G=0.000000,B=0.000000,A=1.000000)" :TraceHitColor "(R=0.000000,G=1.000000,B=0.000000,A=1.000000)" :DrawTime 2e+01))
          (if _returnvalue_2
            (LocalVariables|SetHit _outhit)
            (LocalVariables|SetbCanSpawnAirbag true)
            (Utilities|SpawnAirbag (LocalVariables|GetHit))))
        (:then_3
          ;[1 0.2 0] Debug
          (if (Variables|GetbDebug)
            (Utilities|FlowControl|Sequence
              (:then_0
                (Rendering|Debug|DrawDebugSphere (LocalVariables|GetAttachParentLocation) 3e+01 12 "(R=1.000000,G=0.494174,B=0.000000,A=1.000000)" 15.0) ; Player Location
                (Rendering|Debug|DrawDebugSphere (LocalVariables|GetProjectileLocation) 25.0 12 _returnvalue 15.0) ; Projectile Location
                (Rendering|Debug|DrawDebugSphere (LocalVariables|GetTraceStart) 25.0 12 _returnvalue 15.0) ; Trace Start Location
                (Rendering|Debug|DrawDebugSphere (LocalVariables|GetTraceEnd) 25.0 12 _returnvalue 15.0) ; Trace End Location
                (Rendering|Debug|DrawDebugLine (LocalVariables|GetTraceStart) (LocalVariables|GetTraceEnd) _returnvalue 15.0) ; Trace Start to Trace End
              )
              (:then_1
                (if _bblockinghit
                  (bind _validspawn (CallFunction|IsTraceValidforSpawn _hitcomponent _hitactor (LocalVariables|GetHit)))
                  (Rendering|Debug|DrawDebugSphere _location 25.0 12 (Math|Color|SelectColor "(R=0.000000,G=1.000000,B=0.000000,A=1.000000)" "(R=1.000000,G=0.000000,B=0.037938,A=1.000000)" _validspawn) 15.0) ; Hit Location
                )))))))))

(fn UpdateRelativeLocationToPlayer
  (params (Relative%20Z%20Pos%20To%20Player :type Float%20%28double-precision%29))
  (bind (_returnvalue_x _returnvalue_y) (Math|Vector|BreakVector (* (Math|Vector|RotateVector (Utilities|GetPlayerFlatVelocity) (Math|Rotator|InvertRotator (Transformation|GetActorRotation (Actor|GetAttachParentActor)))) (* Relative%20Z%20Pos%20To%20Player (ScalableFloat|GetValueatLevel (Variables|Default|GetLeadDistance))))))
  (Transformation|SetActorRelativeLocation 0 (Math|Vector|MakeVector _returnvalue_x _returnvalue_y (* Relative%20Z%20Pos%20To%20Player -2.6e+03)) false false))

(fn ConstructionScript
  (|CallParentFunction))
