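; GA_AppleSun_DestroyOverlappingObjects is a passive gameplay ability for the AppleSun item that, by its name, destroys objects overlapping it.
; On activation it runs a target-selection task and applies the gameplay effect container to the targets found.
; That application is wrapped in a structural support system actor-removal batch, so resulting building removals are batched.
; Target selection's custom source transform is overridden to return the avatar actor's transform when valid.
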(event Ability|EventActivateAbility
  (bind _buildingstructuralsupportsystem (Building|GetBuildingStructuralSupportSystem))
  (bind (_asynctaskproxy _targetdata _applicationtag) (Ability|Tasks|DoTargetSelection "None" "(TagName=\"Targeting.ApplicationTag.0\")" true)
    (:Targeted
      (Building|StartActorRemovalBatch _buildingstructuralsupportsystem)
      (Ability|ApplyGameplayEffectContainer _targetdata _applicationtag 0)
      (Building|StopActorRemovalBatch _buildingstructuralsupportsystem))))

(event Ability|EventOnEndAbility
  (params (bWasCancelled :type Boolean)))

(event Ability|EventFailedToActivatePassiveAbility
  (params (ActorInfo :type Gameplay%20Ability%20Actor%20Info%20Structure))
  (|CallParentFunction))

(fn GetCustomAbilitySourceTransform
  (bind _returnvalue (Ability|GetAvatarActorfromActorInfo))
  (Utilities|IsValid _returnvalue
    (:Is%20Valid
      (return (Transformation|GetActorTransform _returnvalue)))))
