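; Passive AppleSun gameplay ability that removes its granting fall-immunity effect once the pawn lands, swims or starts riding.
; Movement-mode changes and a looping vehicle ground check remove one stack, while a riding tag and a looping walking/swimming fallback remove all.
; Removal plays a landing cue, then waits briefly so immunity outlasts the damage frame.
; Ending the ability resets the pawn's ignore-fall-damage flag and clears all stacks.
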(event Custom|TestWalking
  ; FORT-956784: Capture when Player is walking and the Binding somehow missed the change in MovementMode
  (if (or (Movement|IsMovingonGround (Pawn|GetMovementComponent (Variables|Default|GetFort%20Pawn))) (Movement|IsSwimming (Pawn|GetMovementComponent (Variables|Default|GetFort%20Pawn))))
    (CallFunction|RemoveAllGEStacks)))

(event Ability|EventActivateAbility
  (bind _asfort_pawn (Utilities|Casting|CastToFortPawn (Ability|GetAvatarActorfromActorInfo))
    (:then
      (bind _output_get (Variables|Default|SetFort%20Pawn _asfort_pawn))
      (Utilities|FlowControl|Sequence
        (:then_0
          (Character|BindEventtoMovementModeChangedDelegate _output_get (delegate Custom|OnMovementModeChanged))
          (bind _returnvalue (Utilities|Time|SetTimerbyEvent (EventDispatchers|CreateEvent :!event Custom|VehicleLandedCheck) 0.2 true))
          (Variables|Default|SetVehicleOnGroundTimer _returnvalue)
          ; FORT-956784: Capture when Player is walking and the Binding somehow missed the change in MovementMode
          (Utilities|Time|SetTimerbyEvent (delegate Custom|TestWalking) 0.75 true true))
        (:then_1
          (Ability|Tasks|WaitGameplayTagAdd (Variables|Default|GetTagIsRiding)
            (:Added
              (CallFunction|RemoveAllGEStacks) ; Remove all stacks since riding affects landing events
            )) ; If the player start to ride creature we have to stop the Immunity too
        )))
    (:CastFailed
      (Ability|EndAbility))))

(event Custom|OnMovementModeChanged
  (params (Character :type Character%20Object%20Reference) (PrevMovementMode :type EMovementMode%20Enum :default "MOVE_None") (PreviousCustomMode :type Byte))
  (bind _charactermovement (Variables|/Script/Engine.Character|GetCharacterMovement Character))
  (bind _movementmode (Variables|/Script/Engine.CharacterMovementComponent|GetMovementMode _charactermovement))
  (if (Math|Boolean|ORBoolean (Utilities|Enum|Equal(Enum) _movementmode "MOVE_Walking") (Utilities|Enum|Equal(Enum) _movementmode "MOVE_NavWalking") (Utilities|Enum|Equal(Enum) _movementmode "MOVE_Swimming"))
    _
    (elif (Utilities|Enum|Equal(Enum) _movementmode "MOVE_Custom")
      (Utilities|FlowControl|Delay 0.0) ; Wait until the next frame so tag gets applied
      (if (GameplayTags|HasAnyMatchingGameplayTags (Variables|Default|GetFort%20Pawn) (GameplayTags|MakeLiteralGameplayTagContainer "(GameplayTags=((TagName=\"MovementMode.SurfaceSwimming\"),(TagName=\"MovementMode.Sliding\")))"))
        _
        (elif (Movement|IsMovingonGround (Pawn|GetMovementComponent (Variables|Default|GetFort%20Pawn)))
          _))
      (CallFunction|RemoveGE))))

(event Ability|EventFailedToActivatePassiveAbility
  (params (ActorInfo :type Gameplay%20Ability%20Actor%20Info%20Structure))
  (|CallParentFunction))

(event Custom|VehicleLandedCheck
  (bind _vehicle_occupied (Utilities|IsinVehicle (Variables|Default|GetFort%20Pawn))
    (:true
      (Utilities|IsValid (Utilities|ToObject(Interface) _vehicle_occupied)
        (:Is%20Valid
          (if (Class|FortVehicleInterface|IsTouchingGround _vehicle_occupied)
            (Utilities|Time|ClearandInvalidateTimerbyHandle (Variables|Default|GetVehicleOnGroundTimer))
            (Utilities|FlowControl|Delay 0.35) ; Wait just a little to give time to send vehicle landed stat events with appllied tags 
            (CallFunction|RemoveGE)))))))

(event Ability|EventOnEndAbility
  (params (bWasCancelled :type Boolean))
  ;[1 0.4306 0] OnEndAbility() and ServerRemove()
  (|CallParentFunction bWasCancelled)
  (Utilities|Time|ClearandInvalidateTimerbyHandle (Variables|Default|GetVehicleOnGroundTimer))
  (bind _fort_pawn (Variables|Default|GetFort%20Pawn)
    (:then
      (Variables|/Script/FortniteGame.FortPawn|SetbIgnoreNextFallingDamage false _fort_pawn) ; Temporary fix, because this variable isn't getting reset\nwhen player hits water.\n(Can be removed when that code fix is in)
      (Character|UnbindEventfromMovementModeChangedDelegate _fort_pawn (EventDispatchers|CreateEvent :!event Custom|OnMovementModeChanged)))
    (:else
      _))
  (CallFunction|RemoveGrantedbyEffectStacks -1) ; Make sure GE stack is removed if ability is cancelled
)

(event Custom|RemoveGE
  ;[1 0.4306 0] OnEndAbility() and ServerRemove()
  (Ability|AddGameplayCueToOwner (Variables|Default|GetPlayer%20Landing%20GC))
  (Utilities|FlowControl|Delay 0.2) ; Would ideally use DelayUntilNextFrame\nwhen it becomes available\n\nMake sure FallDamage immunity tag doesn't get\nremoved on same frame damage is applied
  (CallFunction|RemoveGrantedbyEffectStacks 1))

(event Custom|RemoveAllGEStacks
  ;[1 0.4306 0] OnEndAbility() and ServerRemove()
  (Ability|AddGameplayCueToOwner (Variables|Default|GetPlayer%20Landing%20GC))
  (Utilities|FlowControl|Delay 0.2) ; Would ideally use DelayUntilNextFrame\nwhen it becomes available\n\nMake sure FallDamage immunity tag doesn't get\nremoved on same frame damage is applied
  (CallFunction|RemoveGrantedbyEffectStacks -1))
