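# Surface material for the AppleSun gadget's bag, animating its inflation wobble, bounce and deflation entirely through World Position Offset.
# Timestamp parameters drive eased sine wobble and bounce curves that scale pivot-relative local positions, with sphere masks spreading timing outward from pivot or bounce point.
# Deflate parameters squash and jiggle the mesh and blend in a tiled wrinkle normal.
# Water displacement is added when a WaterIndex is set.
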Material /Game/Athena/Items/Gameplay/Passives/AppleSun/M_AppleSun
graph {
 properties {bUseMaterialAttributes=true, bUsedWithNanite=true, pixelDepthOffsetMode=PDOM_Legacy, usedShadingModels=Default Lit}
 nodes [
  n1 = TextureSampleParameter2D() {
   desc="R = SpecMask G = Metallic B = Roughness"
   group=Specular
   parameterName=SpecularMasks
   samplerSource=SSM_Wrap_WorldGroupSettings
   samplerType=SAMPLERTYPE_Masks
   texture=/Game/Gadgets/Assets/AppleSun/Textures/T_Item_Apple_Sun_Bag_S
  }
  n2 = Multiply(A=TextureCoordinate(), B=ScalarParameter() {defaultValue=1.6, parameterName=DeflateNormalTile})
  n3 = Multiply(
   A=ScalarParameter() {parameterName=Deflate}
   B=Multiply(
    A=TextureSample(UVs=@n2) {samplerType=SAMPLERTYPE_Normal, texture=/FortWater/Textures/02_WaterFoam_Veroni_base_N}.RGB
    B=Constant3Vector() {constant=<rgba 1.25 1.25 0 0>}
   )
  )
  n4 = SphereMask(
   A=Subtract(
    A=WorldPosition().XYZ
    B=MaterialFunctionCall() {materialFunction=&ObjectPivotPoint}.Object Pivot Location
   )
   B=Constant()
   Radius=ScalarParameter() {defaultValue=500, parameterName=DeflateSphereRadius}
  ) {hardnessPercent=0}
  n5 = Multiply(A=Power(Base=@n4) {constExponent=1.5}, B=ScalarParameter() {defaultValue=0.5, parameterName=OuterDeflateTimeOffset}) {constB=0.25}
  n6 = ScalarParameter() {parameterName=Deflate}
  n7 = Clamp(None=Add(A=@n6, B=Multiply(A=@n6, B=@n5)))
  n8 = Power(Base=@n7, Exp=ScalarParameter() {defaultValue=0.5, parameterName=DeflateTimePower})
  n9 = VectorParameter() {parameterName=DeflateJiggle, defaultValue=<rgba 0.1 0.1 0.18 0>}
  n10 = Sine(None=@n7) {period=0.15}
  n11 = Transform(None=Subtract(
    A=WorldPosition().XYZ
    B=MaterialFunctionCall() {materialFunction=&ObjectPivotPoint}.Object Pivot Location
   )) {transformSourceType=TRANSFORMSOURCE_World, transformType=TRANSFORM_Local}
  n12 = Multiply(A=Multiply(A=OneMinus(None=@n8), B=LinearInterpolate(B=@n9.RGB, Alpha=@n10) {constB=0}), B=@n11)
  n13 = Multiply(
   A=@n11
   B=LinearInterpolate(A=Constant3Vector() {constant=<rgba -0.5 -0.5 0 1>}, B=Constant3Vector() {constant=<rgba 0 0 -0.99 0>}, Alpha=@n4)
  )
  n14 = ScalarParameter() {parameterName=DeflateStart}
  n15 = Time() {period=5}
  n16 = ScalarParameter() {parameterName=WobbleStartTime}
  n17 = Clamp(None=Divide(A=Subtract(A=@n15, B=@n16), B=ScalarParameter() {defaultValue=1.6, group=Timestamp, parameterName=WobbleDuration}))
  n18 = Multiply(A=@n17) {constB=3}
  n19 = Saturate(None=Power(Base=Saturate(None=@n18)) {constExponent=0.5})
  n20 = SphereMask(
   A=ComponentMask(None=WorldPosition().XYZ) {g=true, r=true}
   B=ComponentMask(None=MaterialFunctionCall() {materialFunction=&ObjectPivotPoint}.Object Pivot Location) {g=true, r=true}
   Radius=LinearInterpolate(A=ScalarParameter() {parameterName=RadiusStart}, B=ScalarParameter() {defaultValue=900, parameterName=RadiusEnd}, Alpha=@n19)
  ) {hardnessPercent=0}
  n21 = Multiply(A=OneMinus(None=@n20), B=ScalarParameter() {defaultValue=-0.15, parameterName=OuterWobbleTimeOffset}) {constB=0.1}
  n22 = Transform(None=Subtract(
    A=WorldPosition().XYZ
    B=MaterialFunctionCall() {materialFunction=&ObjectPivotPoint}.Object Pivot Location
   )) {transformSourceType=TRANSFORMSOURCE_World, transformType=TRANSFORM_Local}
  n23 = Multiply(
   A=Add(A=@n14, B=Multiply(A=@n14, B=@n21))
   B=Multiply(A=VectorParameter() {parameterName=DeflateScale, defaultValue=<rgba 0.05 0.05 -0.15 0>}.RGB, B=Multiply(A=@n22, B=@n20))
  )
  n24 = LinearInterpolate(A=Constant3Vector() {constant=<rgba 1.5 1.5 1 0>}, Alpha=@n20) {constA=1.5}
  n25 = Multiply(A=@n22, B=LinearInterpolate(Alpha=@n19) {constA=-1, constB=0})
  n26 = LinearInterpolate(Alpha=@n20) {constA=0.7}
  n27 = Saturate(None=Power(Base=@n17, Exp=ScalarParameter() {defaultValue=0.3, parameterName=GrowTimePower}) {constExponent=0.2})
  n28 = VectorParameter() {parameterName=WobbleMult1, defaultValue=<rgba 0.5 0.5 1 1>}
  n29 = VectorParameter() {parameterName=WobbleMult2, defaultValue=<rgba -0.5 -0.5 5 1>}
  n30 = Power(Base=@n17, Exp=ScalarParameter() {defaultValue=1.5, parameterName=WobbleTransitionPower}) {constExponent=1.5}
  n31 = ConstantBiasScale(None=Sine(None=Add(A=@n21, B=@n30)) {period=0.25})
  n32 = ScalarParameter() {defaultValue=1, parameterName=WobbleStrength}
  n33 = ConstantBiasScale(None=OneMinus(None=Sine(None=@n27)))
  n34 = Multiply(
   A=Multiply(A=@n22, B=LinearInterpolate(Alpha=@n27) {constA=-1, constB=0}) {constB=-1}
   B=Multiply(A=LinearInterpolate(A=@n28.RGB, B=@n29.RGB, Alpha=@n31), B=LinearInterpolate(A=@n32, Alpha=@n33) {constA=1, constB=0})
  )
  n35 = Multiply(
   A=Add(A=Multiply(A=@n24, B=@n25) {constA=1}, B=Multiply(A=@n26, B=@n34)) {constA=1, constB=0}
   B=ScalarParameter() {defaultValue=1, parameterName=WPOStrength}
  )
  n36 = SphereMask(
   A=ComponentMask(None=WorldPosition().XYZ) {g=true, r=true}
   B=ComponentMask(None=VectorParameter() {parameterName=BouncePosition, defaultValue=<rgba 400 400 0 0>}.RGB) {g=true, r=true}
   Radius=ScalarParameter() {defaultValue=800, parameterName=BounceRadius}
  ) {hardnessPercent=0}
  n37 = Transform(None=Subtract(
    A=WorldPosition().XYZ
    B=MaterialFunctionCall() {materialFunction=&ObjectPivotPoint}.Object Pivot Location
   )) {transformSourceType=TRANSFORMSOURCE_World, transformType=TRANSFORM_Local}
  n38 = Time() {period=5}
  n39 = ScalarParameter() {parameterName=BounceStartTime}
  n40 = Clamp(None=Divide(A=Subtract(A=@n38, B=@n39), B=ScalarParameter() {defaultValue=1.25, group=Timestamp, parameterName=BounceDuration}))
  n41 = Saturate(None=Power(Base=@n40, Exp=ScalarParameter() {defaultValue=0.4, parameterName=BounceTimePower}) {constExponent=0.2})
  n42 = LinearInterpolate(Alpha=@n41) {constA=-1, constB=0}
  n43 = Add(A=Multiply(A=OneMinus(None=@n36), B=ScalarParameter() {defaultValue=-0.15, parameterName=OuterWobbleTimeOffset}) {constB=0.1}, B=@n40)
  n44 = LinearInterpolate(
   A=VectorParameter() {parameterName=BounceMult1, defaultValue=<rgba 0.1 0.1 -1.2 1>}.RGB
   B=VectorParameter() {parameterName=BoucneMult2, defaultValue=<rgba -0.22 -0.22 1.4 0>}.RGB
   Alpha=ConstantBiasScale(None=Sine(None=@n43) {period=0.25})
  )
  n45 = ScalarParameter() {defaultValue=1, parameterName=WobbleStrength}
  n46 = OneMinus(None=Sine(None=@n41))
  n47 = Multiply(A=ComponentMask(None=@n44) {g=true, r=true}, B=LinearInterpolate(A=@n45, Alpha=ConstantBiasScale(None=@n46)))
  n48 = Multiply(A=ComponentMask(None=@n44) {b=true}, B=LinearInterpolate(A=@n45, Alpha=@n41) {constB=0})
  n49 = Multiply(A=LinearInterpolate(Alpha=@n36) {constA=0.5}, B=Multiply(A=Multiply(A=@n37, B=@n42), B=AppendVector(A=@n47, B=@n48))) {constA=1}
  n50 = ScalarParameter() {defaultValue=1, parameterName=BounceIntensity}
  n51 = Add(A=Multiply(A=@n8, B=Add(A=@n12, B=@n13)), B=Add(A=Add(A=@n23, B=@n35), B=Multiply(A=@n49, B=@n50)))
  n52 = ScalarParameter() {defaultValue=-1, parameterName=WaterIndex}
  n53 = Multiply(
   A=MaterialFunctionCall!(Ignore XY Offsets, Position, WaterBodyIndex=@n52, Depth=ScalarParameter() {parameterName=WaterDepth}) {materialFunction=/FortWater/Materials/Functions/WaterParticleDisplacement}.WPO
   B=ScalarParameter() {defaultValue=0.7, parameterName=WaterInfluence}
  )
  n54 = Constant()
  n55 = SetMaterialAttributes!(
   MaterialAttributes
   Base Color=TextureSampleParameter2D() {parameterName=Diffuse, texture=/Game/Gadgets/Assets/AppleSun/Textures/T_Item_Apple_Sun_Bag_D}.RGB
   Metallic=Constant()
   Specular=@n1.R
   Roughness=LinearInterpolate(
    A=ScalarParameter() {group=Roughness, parameterName=SpecRoughnessMin}
    B=ScalarParameter() {defaultValue=1, group=Roughness, parameterName=SpecRoughnessMax}
    Alpha=@n1.B
   )
   Normal=Add(
    A=TextureSampleParameter2D() {parameterName=Normal, samplerType=SAMPLERTYPE_Normal, texture=/Game/Gadgets/Assets/AppleSun/Textures/T_Item_Apple_Sun_Bag_N}.RGB
    B=@n3
   )
   World Position Offset=Add(A=Transform(None=@n51) {transformSourceType=TRANSFORMSOURCE_Local}, B=If(A=@n52, "A > B"=@n53, "A == B"=@n54, "A < B"=@n54) {constB=-1})
   Subsurface Color
   Emissive Color
  ) {attributeSetTypes [
    #1
    #2
    #3
    #4
    #5
    #6
    #7
    #8
   ]}
  MaterialFunctionCall(MA In=@n55) {materialFunction=&MF_CreativeMoveTool}
  StaticSwitchParameter(True=@n3, False=Constant()) {defaultValue=true, parameterName=Deflate}
  Time() {bOverride_Period=true, period=1.25}
  Time() {bOverride_Period=true, period=1.6}
 ]
}
